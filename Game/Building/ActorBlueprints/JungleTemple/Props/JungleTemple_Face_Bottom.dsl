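; Jungle temple bottom face building prop whose mouth can show a waterfall, a flame with point light, or neither, with waterfall taking priority over fire.
; Construction script sets visibility, auto-activation and light, plus waterfall offset, foam color and a tinted dynamic water material.
; BeginPlay starts or stops the matching particles and audio, and EventDestroyed fires a water or flame destruction effect.
; All effect branching is gated by CanExecuteCosmeticEvents.
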(event EventBeginPlay
  (bind _p_waterfall_junglemouth (Variables|Default|GetP_WaterFall_JungleMouth))
  (bind _wateraudio (Variables|Default|GetWaterAudio))
  (bind _fireaudio (Variables|Default|GetFireAudio))
  (bind _pointlight (Variables|Default|GetPointLight))
  (bind _p_athena_jungle_mouth_flame (Variables|Default|GetP_Athena_Jungle_Mouth_Flame))
  (|CallParentFunction)
  (Networking|CanExecuteCosmeticEvents
    (:True
      ;[1 1 1] Water
      (if (Variables|FX|GetUseWaterfall)
        (Components|Activation|Activate _p_waterfall_junglemouth false)
        (Audio|Components|Audio|Play _wateraudio 0.0)
        (else
          (Components|Activation|Deactivate _p_waterfall_junglemouth)
          (Audio|Components|Audio|Stop _wateraudio)
          ;[1 1 1] Fire
          (if (Variables|FX|GetUseFire)
            (Components|Activation|Activate _p_athena_jungle_mouth_flame false)
            (Audio|Components|Audio|Play _fireaudio 0.0)
            (Rendering|Components|Light|SetIntensity _pointlight 6905.066895)
            (else
              (Components|Activation|Deactivate _p_athena_jungle_mouth_flame)
              (Audio|Components|Audio|Stop _fireaudio)
              (Rendering|Components|Light|SetIntensity _pointlight 0.0))))))))

(event EventDestroyed
  (|CallParentFunction)
  (Networking|CanExecuteCosmeticEvents
    (:True
      (if (Variables|FX|GetUseWaterfall)
        ;[1 1 1] Destroy Water
        (Transformation|AddLocalOffset (Variables|Default|GetP_JungleMouthDestroy_Water) (Variables|FX|GetSetLocalOffset) false false)
        (Components|Activation|Activate (Variables|Default|GetP_JungleMouthDestroy_Water) false)
        (Effects|Components|ParticleSystem|SetVectorParameter (Variables|Default|GetP_JungleMouthDestroy_Water) "FoamColor" (Math|Conversions|ToVector(LinearColor) (Variables|FX|GetFoam%20FX%20Color)))
        (elif (Variables|FX|GetUseFire)
          ;[1 1 1] Destroy Fire
          (Components|Activation|Activate (Variables|Default|GetP_Athena_Jungle_MouthDestroy_Flame) false))))))

(fn OnRep_UseWaterfall)

(fn ConstructionScript
  (bind _p_athena_jungle_mouth_flame (Variables|Default|GetP_Athena_Jungle_Mouth_Flame))
  (bind _pointlight (Variables|Default|GetPointLight))
  (|CallParentFunction)
  (Rendering|Components|Light|SetIntensity (Variables|Default|GetPointLight) 0.0)
  (Networking|CanExecuteCosmeticEvents
    (:True
      (if (Variables|FX|GetUseWaterfall)
        ;[1 1 1] Water!
        (Rendering|SetVisibility (Variables|Default|GetP_WaterFall_JungleMouth) true false)
        (Components|Activation|SetAutoActivate (Variables|Default|GetP_WaterFall_JungleMouth) true)
        (Transformation|AddLocalOffset (Variables|Default|GetP_WaterFall_JungleMouth) (Variables|FX|GetSetLocalOffset) false false)
        (Effects|Components|ParticleSystem|SetVectorParameter (Variables|Default|GetP_WaterFall_JungleMouth) "FoamColor" (Math|Conversions|ToVector(LinearColor) (Variables|FX|GetFoam%20FX%20Color)))
        (bind _returnvalue (Rendering|Material|CreateDynamicMaterialInstance (Variables|Default|GetSM_JungleTempleMouthWater) 0 (Rendering|Material|GetMaterial (Variables|Default|GetSM_JungleTempleMouthWater) 0) "None" :!class /Script/Engine.PrimitiveComponent))
        (bind _output_get (Variables|FX|SetWaterMID _returnvalue))
        (Rendering|Material|SetMaterial (Variables|Default|GetSM_JungleTempleMouthWater) 0 _output_get)
        (Rendering|Material|SetVectorParameterValue (Variables|FX|GetWaterMID) "BaseColor" (Variables|FX|GetWaterDarkColor) :!class /Script/Engine.MaterialInstanceDynamic)
        (Rendering|Material|SetVectorParameterValue (Variables|FX|GetWaterMID) "FoamColor" (Variables|FX|GetWater%20Foam%20Color) :!class /Script/Engine.MaterialInstanceDynamic)
        (else
          (Rendering|SetVisibility (Variables|Default|GetSM_JungleTempleMouthWater) false true)
          ;[1 1 1] FIRE!
          (if (Variables|FX|GetUseFire)
            (Rendering|SetVisibility _p_athena_jungle_mouth_flame true false)
            (Components|Activation|SetAutoActivate _p_athena_jungle_mouth_flame true)
            (Rendering|Components|Light|SetIntensity _pointlight 6905.066895)
            (else
              (Components|Activation|SetAutoActivate (Variables|Default|GetP_Athena_Jungle_Mouth_Flame) false)
              (Rendering|SetVisibility (Variables|Default|GetSM_FlameMouth) false false)
              (Rendering|Components|Light|SetIntensity _pointlight 0.0))))))))
