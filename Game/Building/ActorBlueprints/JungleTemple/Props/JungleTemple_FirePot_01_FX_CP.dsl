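; Construction script for a jungle temple fire pot prop Blueprint with fire effects.
; After calling the parent construction script, it sets the LightBrightness float parameter on the Brazier_Fire particle component from the LightBrightness variable.
; This lets the fire effect's light brightness follow that variable when the actor is constructed.
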(fn ConstructionScript
  (|CallParentFunction)
  (Effects|Components|ParticleSystem|SetFloatParameter (Variables|Default|GetBrazier_Fire) "LightBrightness" (Variables|Default|GetLightBrightness)))
