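; Jungle temple fire pot prop Blueprint, a variant named NoShadow, that trims its fire effects by platform when play begins.
; BeginPlay runs the parent logic, then checks whether the platform is Nintendo Switch or mobile.
; On those platforms it destroys the BloomPlane component; otherwise it disables the "Light" emitter on the NS_Brazier_Fire effect.
; The construction script only calls the parent.
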(event EventBeginPlay
  (|CallParentFunction)
  (if (or (TouchPlatform|IsNintendoSwitch) (TouchPlatform|IsMobileGame))
    (Components|DestroyComponent (Variables|Default|GetBloomPlane))
    (Effects|Components|ParticleSystem|SetEmitterEnable (Variables|Default|GetNS_Brazier_Fire) "Light" false)))

(fn ConstructionScript
  (|CallParentFunction))
