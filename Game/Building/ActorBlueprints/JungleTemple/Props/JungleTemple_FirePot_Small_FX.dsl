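; Blueprint for a small Jungle Temple fire pot building prop with brazier fire effects.
; On BeginPlay it runs the parent logic, then disables the "Light" emitter on its NS_Brazier_Fire particle component.
; It supplies three single-component arrays for dynamic destruction: the static mesh, the bloom plane and the static mesh again.
; The construction script only calls its parent.
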(event EventBeginPlay
  (|CallParentFunction)
  (Effects|Components|ParticleSystem|SetEmitterEnable (Variables|Default|GetNS_Brazier_Fire) "Light" false))

(fn GetDynamicDestructionStaticMeshComponents_0
  (return (Utilities|Array|MakeArray (Variables|Debug|GetStaticMeshComponent)) (Utilities|Array|MakeArray (Variables|Default|GetBloomPlane)) (Utilities|Array|MakeArray (Variables|Debug|GetStaticMeshComponent))))

(fn ConstructionScript
  (|CallParentFunction))
