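; Small Jungle Temple fire pot prop Blueprint with brazier fire particles and a bloom plane.
; The construction script runs the parent logic, then passes the LightBrightness variable to the Brazier_Fire particle system's float parameter of the same name.
; GetDynamicDestructionStaticMeshComponents returns three single-element component arrays: the static mesh in the first and third, the bloom plane in the second.
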(fn GetDynamicDestructionStaticMeshComponents_0
  (return (Utilities|Array|MakeArray (Variables|Debug|GetStaticMeshComponent)) (Utilities|Array|MakeArray (Variables|Default|GetBloomPlane)) (Utilities|Array|MakeArray (Variables|Debug|GetStaticMeshComponent))))

(fn ConstructionScript
  (|CallParentFunction)
  (Effects|Components|ParticleSystem|SetFloatParameter (Variables|Default|GetBrazier_Fire) "LightBrightness" (Variables|Default|GetLightBrightness)))
