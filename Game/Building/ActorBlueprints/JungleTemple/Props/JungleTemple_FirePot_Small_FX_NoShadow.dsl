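; Building prop Blueprint for a small Jungle Temple fire pot with FX, named as a no-shadow variant.
; It overrides the query for dynamic destruction static mesh components, returning three single-element arrays: the static mesh component, the bloom plane, and the static mesh component again.
; The construction script only calls the parent, and the graphs shown contain no fire or shadow logic.
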(fn GetDynamicDestructionStaticMeshComponents_0
  (return (Utilities|Array|MakeArray (Variables|Debug|GetStaticMeshComponent)) (Utilities|Array|MakeArray (Variables|Default|GetBloomPlane)) (Utilities|Array|MakeArray (Variables|Debug|GetStaticMeshComponent))))

(fn ConstructionScript
  (|CallParentFunction))
